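# Physics asset for the JollyTroll backpack cosmetic, defining collision bodies and joints for its dyn_ bones: neck, head, hat, two headband tails, two ears and a hanging metal ball.
# Each bone has one capsule, chained from the neck through the head to the hat, ears and headbands, with the ball connector branching off the neck.
# Head, hat and chain-root joints rotate freely, while inner joints use angular limits: very stiff for ears, looser toward headband tips and ball.
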PhysicsAsset /BRCosmetics/Accessories/FORT_Backpacks/Backpack_M_MED_JollyTroll/Meshes/M_MED_JollyTroll_Pack_Physics
bodies {
 dyn_hat [{shapeType=Capsule, length=3.4097, radius=2.3689, shapeName=dyn_hat_capsule, center=<0 -0.3267 3.2755>, rotation=<rot 0 0 94.2835>}]
 dyn_head [{
   shapeType=Capsule
   length=4.75
   radius=3.8717
   shapeName=dyn_head_capsule
   center=<0.083 -0.4705 4.2199>
   rotation=<rot 83.4151 -84.989 5.2141>
  }]
 dyn_headband_a_01 [{shapeType=Capsule, length=1.2904, radius=0.8, shapeName=dyn_headband_a_01_capsule, center=<1.1542 0.0541 0.1953>, rotation=<rot 90 0 0>}]
 dyn_headband_a_02 [{shapeType=Capsule, length=1.2752, radius=0.8, shapeName=dyn_headband_a_02_capsule, center=<0.5 0 0>, rotation=<rot 90 0 0>}]
 dyn_headband_a_03 [{shapeType=Capsule, length=1.8926, radius=0.8, shapeName=dyn_headband_a_03_capsule, center=<1.334 0.3701 -0.0419>, rotation=<rot 90 0 0>}]
 dyn_headband_a_root [{shapeType=Capsule, length=0.505, radius=0.505, shapeName=dyn_headband_a_root_capsule, rotation=<rot 0 -90 0>}]
 dyn_headband_b_01 [{shapeType=Capsule, length=1.3285, radius=0.8, shapeName=dyn_headband_b_01_capsule, center=<1.2701 -0.029 0.0222>, rotation=<rot 90 0 0>}]
 dyn_headband_b_02 [{shapeType=Capsule, length=1.0828, radius=0.8, shapeName=dyn_headband_b_02_capsule, center=<0.3969 0 0>, rotation=<rot 90 0 0>}]
 dyn_headband_b_03 [{shapeType=Capsule, length=1.6663, radius=0.8, shapeName=dyn_headband_b_03_capsule, center=<1.0067 0 0>, rotation=<rot 90 0 0>}]
 dyn_headband_b_root [{shapeType=Capsule, length=0.505, radius=0.505, shapeName=dyn_headband_b_root_capsule, rotation=<rot 0 -90 0>}]
 dyn_l_ear_01 [{
   shapeType=Capsule
   length=0.505
   radius=0.505
   shapeName=dyn_l_ear_01_capsule
   center=<1.4699 0.0875 0.6973>
   rotation=<rot 0 -148.9613 -83.7213>
  }]
 dyn_l_ear_02 [{shapeType=Capsule, length=1.776, radius=1.25, shapeName=dyn_l_ear_02_capsule, center=<0.0556 0.0176 0.1698>, rotation=<rot 0 90 -90>}]
 dyn_l_ear_03 [{shapeType=Capsule, length=1.8819, radius=0.7858, shapeName=dyn_l_ear_03_capsule, center=<1.717 -0.3099 -0.0382>, rotation=<rot 0 -90 -90>}]
 dyn_l_ear_root [{shapeType=Capsule, length=0.505, radius=0.505, shapeName=dyn_l_ear_root_capsule, rotation=<rot 0 -90 0>}]
 dyn_metal_ball [{shapeType=Capsule, length=0.25, radius=1.5388, shapeName=dyn_metal_ball_capsule, center=<1.8316 0.0052 0>, rotation=<rot 0 90 90>}]
 dyn_metal_ball_connector [{
   shapeType=Capsule
   length=0.9788
   radius=0.9717
   shapeName=dyn_metal_ball_connector_capsule
   center=<0.5506 -0.3059 0.0452>
   rotation=<rot 0 94.5122 -86.0494>
  }]
 dyn_metal_ball_connector_root [{shapeType=Capsule, length=0.505, radius=0.505, shapeName=dyn_metal_ball_connector_root_capsule, rotation=<rot 0 -90 0>}]
 dyn_neck_bottom [{
   shapeType=Capsule
   length=4.8968
   radius=4.8227
   shapeName=dyn_neck_bottom_capsule
   center=<-0.0408 -1.0004 -1.3196>
   rotation=<rot 64.6871 89.3234 -90.4967>
  }]
 dyn_r_ear_01 [{
   shapeType=Capsule
   length=0.505
   radius=0.505
   shapeName=dyn_r_ear_01_capsule
   center=<-1.4699 -0.0875 -0.6973>
   rotation=<rot 0 -148.9613 -83.7213>
  }]
 dyn_r_ear_02 [{shapeType=Capsule, length=1.776, radius=1.25, shapeName=dyn_r_ear_02_capsule, center=<-0.0556 -0.0176 -0.0644>, rotation=<rot 0 90 -90>}]
 dyn_r_ear_03 [{shapeType=Capsule, length=1.8819, radius=0.7858, shapeName=dyn_r_ear_03_capsule, center=<-1.717 0.3099 0.0382>, rotation=<rot 0 -90 -90>}]
 dyn_r_ear_root [{shapeType=Capsule, length=0.505, radius=0.505, shapeName=dyn_r_ear_root_capsule, rotation=<rot 0 -90 0>}]
}
constraints [
 {bone1Name=dyn_headband_a_01, bone2Name=dyn_headband_a_root, swing1LimitDegrees=15, swing2LimitDegrees=15, twistLimitDegrees=15}
 {bone1Name=dyn_headband_a_02, bone2Name=dyn_headband_a_01, swing1LimitDegrees=15, swing2LimitDegrees=15, twistLimitDegrees=5}
 {bone1Name=dyn_headband_a_03, bone2Name=dyn_headband_a_02, swing1LimitDegrees=25, swing2LimitDegrees=25, twistLimitDegrees=15}
 {bone1Name=dyn_headband_b_01, bone2Name=dyn_headband_b_root, swing1LimitDegrees=15, swing2LimitDegrees=15, twistLimitDegrees=15}
 {bone1Name=dyn_headband_b_02, bone2Name=dyn_headband_b_01, swing1LimitDegrees=25, swing2LimitDegrees=25, twistLimitDegrees=10}
 {bone1Name=dyn_headband_b_03, bone2Name=dyn_headband_b_02, swing1LimitDegrees=50, swing2LimitDegrees=50, twistLimitDegrees=20}
 {bone1Name=dyn_metal_ball_connector, bone2Name=dyn_metal_ball_connector_root, swing1LimitDegrees=5, swing2LimitDegrees=5, twistLimitDegrees=5}
 {bone1Name=dyn_metal_ball, bone2Name=dyn_metal_ball_connector, swing1LimitDegrees=45, swing2LimitDegrees=45, twistLimitDegrees=45}
 {
  bone1Name=dyn_headband_a_root
  bone2Name=dyn_head
  swing1LimitDegrees=45
  swing1Motion=Free
  swing2LimitDegrees=45
  swing2Motion=Free
  twistLimitDegrees=45
  twistMotion=Free
 }
 {
  bone1Name=dyn_headband_b_root
  bone2Name=dyn_head
  swing1LimitDegrees=45
  swing1Motion=Free
  swing2LimitDegrees=45
  swing2Motion=Free
  twistLimitDegrees=45
  twistMotion=Free
 }
 {
  bone1Name=dyn_l_ear_root
  bone2Name=dyn_head
  swing1LimitDegrees=45
  swing1Motion=Free
  swing2LimitDegrees=45
  swing2Motion=Free
  twistLimitDegrees=45
  twistMotion=Free
 }
 {bone1Name=dyn_l_ear_01, bone2Name=dyn_l_ear_root, swing1LimitDegrees=2, swing2LimitDegrees=2, twistLimitDegrees=2}
 {bone1Name=dyn_l_ear_02, bone2Name=dyn_l_ear_01, swing1LimitDegrees=4, swing2LimitDegrees=4, twistLimitDegrees=4}
 {bone1Name=dyn_l_ear_03, bone2Name=dyn_l_ear_02, swing1LimitDegrees=4, swing2LimitDegrees=4, twistLimitDegrees=4}
 {
  bone1Name=dyn_r_ear_root
  bone2Name=dyn_head
  swing1LimitDegrees=45
  swing1Motion=Free
  swing2LimitDegrees=45
  swing2Motion=Free
  twistLimitDegrees=45
  twistMotion=Free
 }
 {bone1Name=dyn_r_ear_01, bone2Name=dyn_r_ear_root, swing1LimitDegrees=2, swing2LimitDegrees=2, twistLimitDegrees=2}
 {bone1Name=dyn_r_ear_02, bone2Name=dyn_r_ear_01, swing1LimitDegrees=4, swing2LimitDegrees=4, twistLimitDegrees=4}
 {bone1Name=dyn_r_ear_03, bone2Name=dyn_r_ear_02, swing1LimitDegrees=4, swing2LimitDegrees=4, twistLimitDegrees=4}
 {
  bone1Name=dyn_hat
  bone2Name=dyn_head
  swing1LimitDegrees=45
  swing1Motion=Free
  swing2LimitDegrees=45
  swing2Motion=Free
  twistLimitDegrees=45
  twistMotion=Free
 }
 {
  bone1Name=dyn_head
  bone2Name=dyn_neck_bottom
  swing1LimitDegrees=45
  swing1Motion=Free
  swing2LimitDegrees=45
  swing2Motion=Free
  twistLimitDegrees=45
  twistMotion=Free
 }
 {
  bone1Name=dyn_metal_ball_connector_root
  bone2Name=dyn_neck_bottom
  swing1LimitDegrees=45
  swing1Motion=Free
  swing2LimitDegrees=45
  swing2Motion=Free
  twistLimitDegrees=45
  twistMotion=Free
 }
]
