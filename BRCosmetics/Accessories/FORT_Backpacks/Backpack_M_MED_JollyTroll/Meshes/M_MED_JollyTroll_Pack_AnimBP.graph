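# Animation Blueprint for the Jolly Troll backpack, which idles looking about and plays a sneeze reaction.
# The sneeze fires on chest-opened or enemy-eliminated input events, after a delay in frontend display, and again on a timer after each sneeze.
# The state machine output feeds a root offset and RigidBody physics; an unconnected testing state machine remains.
# Outside frontend reactivity, updates compute a per-movement-state Gravity Override variable using a deprecated helper.
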AnimBlueprint /BRCosmetics/Accessories/FORT_Backpacks/Backpack_M_MED_JollyTroll/Meshes/M_MED_JollyTroll_Pack_AnimBP
target_skeleton=/Game/Accessories/FORT_Backpacks/Backpack_M_MED_JollyTroll/Meshes/M_MED_JollyTroll_Pack_Skeleton
graphs [
 AnimGraph {nodes [
   Input Pose()
   AnimGraphNode_StateMachine_1 = Idle and Chest_Open() {is=title AnimGraphNode_StateMachine}
   n1 = "Transform (Modify) Bone - Bone: root"(ComponentPose=Local To Component(LocalPose=@AnimGraphNode_StateMachine_1), Translation="0,0,3", Rotation="0,0,0") {translationMode=BMM_Additive, alphaBoolBlend {blendOption=Linear}, boneToModify {boneName=root}, translation {z=3}}
   n2 = RigidBody!(ComponentPose=@n1, Alpha, OverrideWorldGravity:Vector="0,0,-980", ExternalForce) -> (Pose) {
    bClampLinearTranslationLimitToRefPose=true
    bOverrideWorldGravity=true
    overridePhysicsAsset=&/M_MED_JollyTroll_Pack_Physics
    simulationSpace=BaseBoneSpace
    alphaBoolBlend {blendOption=Linear}
    baseBoneRef {boneName=root}
    componentAppliedLinearAccClamp=<3000 3000 3000>
    componentLinearAccScale=<0.4 0.3 0.3>
    componentLinearVelScale=<0.4 0.3 0.3>
   }
   Output Pose(Result=Component To Local(ComponentPose=@n2)) {layerGroup=None}
   AnimGraphNode_StateMachine_0 = Animation Testing() {is=title AnimGraphNode_StateMachine}
  ]}
 Animation Testing {
  entry=State
  path=AnimGraph.AnimGraphNode_StateMachine_0.Animation Testing
  ids=id
  nodes [
   AnimStateNode_0 = State() {is=title AnimStateNode}
   AnimStateNode_1 = State_1() {is=title AnimStateNode}
   AnimStateTransitionNode_0 = AnimStateTransitionNode() {bAutomaticRuleBasedOnSequencePlayerInState=true, crossfadeDuration=0}
   AnimStateTransitionNode_1 = AnimStateTransitionNode() {bAutomaticRuleBasedOnSequencePlayerInState=true, crossfadeDuration=0}
  ]
  links [
   {
    dst=State_1
    rule=AnimGraph.AnimGraphNode_StateMachine_0.Animation Testing.AnimStateTransitionNode_0.Transition
    src=State
    via=AnimStateTransitionNode_0
   }
   {
    dst=State
    rule=AnimGraph.AnimGraphNode_StateMachine_0.Animation Testing.AnimStateTransitionNode_1.Transition
    src=State_1
    via=AnimStateTransitionNode_1
   }
  ]
 }
 State {
  path=AnimGraph.AnimGraphNode_StateMachine_0.Animation Testing.AnimStateNode_0.State
  nodes [
   n1 = Sequence Player 'JollyTroll_Idle_Anim_lookabout_PACK'() {
    bLoopAnimation=false
    sequence=/BRCosmetics/Accessories/FORT_Backpacks/Backpack_M_MED_JollyTroll/Animations/JollyTroll_Idle_Anim_lookabout_PACK
   }
   Output Animation Pose(Result=@n1)
  ]
 }
 State_1 {
  path=AnimGraph.AnimGraphNode_StateMachine_0.Animation Testing.AnimStateNode_1.State_1
  nodes [
   n1 = Sequence Player 'JollyTroll_Sneeze_Anim_PACK'() {bLoopAnimation=false, sequence=/BRCosmetics/Accessories/FORT_Backpacks/Backpack_M_MED_JollyTroll/Animations/JollyTroll_Sneeze_Anim_PACK}
   Output Animation Pose(Result=@n1)
  ]
 }
 Transition {path=AnimGraph.AnimGraphNode_StateMachine_0.Animation Testing.AnimStateTransitionNode_0.Transition, nodes [Result()]}
 Transition {path=AnimGraph.AnimGraphNode_StateMachine_0.Animation Testing.AnimStateTransitionNode_1.Transition, nodes [Result()]}
 Idle and Chest_Open {
  entry=Idle
  path=AnimGraph.AnimGraphNode_StateMachine_1.Idle and Chest_Open
  ids=id
  nodes [
   AnimStateNode_2 = Idle() {
    stateEntered {guid=#1}
    stateFullyBlended {guid=#2}
    stateLeft {guid=#3}
   }
   AnimStateNode_3 = Chest_Open() {
    is=title AnimStateNode
    stateEntered {guid=#4, notifyName=Entered_ChestOpen}
    stateFullyBlended {guid=#5}
    stateLeft {guid=#6}
   }
   AnimStateTransitionNode_2 = AnimStateTransitionNode() {
    crossfadeDuration=0.3
    transitionEnd {guid=#7}
    transitionInterrupt {guid=#8}
    transitionStart {guid=#9}
   }
   AnimStateTransitionNode_3 = AnimStateTransitionNode() {
    bAutomaticRuleBasedOnSequencePlayerInState=true
    crossfadeDuration=0.1
    transitionEnd {guid=#10}
    transitionInterrupt {guid=#11}
    transitionStart {guid=#12}
   }
  ]
  links [
   {
    dst=Idle
    rule=AnimGraph.AnimGraphNode_StateMachine_1.Idle and Chest_Open.AnimStateTransitionNode_3.Transition
    src=Chest_Open
    via=AnimStateTransitionNode_3
   }
   {
    dst=Chest_Open
    rule=AnimGraph.AnimGraphNode_StateMachine_1.Idle and Chest_Open.AnimStateTransitionNode_2.Transition
    src=Idle
    via=AnimStateTransitionNode_2
   }
  ]
 }
 Idle {
  path=AnimGraph.AnimGraphNode_StateMachine_1.Idle and Chest_Open.AnimStateNode_2.Idle
  nodes [Output Animation Pose(Result=Sequence Player 'JollyTroll_Idle_Anim_lookabout_PACK'() {sequence=&JollyTroll_Idle_Anim_lookabout_PACK})]
 }
 Chest_Open {
  path=AnimGraph.AnimGraphNode_StateMachine_1.Idle and Chest_Open.AnimStateNode_3.Chest_Open
  nodes [
   n1 = Sequence Player 'JollyTroll_Sneeze_Anim_PACK'() {bLoopAnimation=false, sequence=&JollyTroll_Sneeze_Anim_PACK}
   Output Animation Pose(Result=@n1)
  ]
 }
 Transition {
  path=AnimGraph.AnimGraphNode_StateMachine_1.Idle and Chest_Open.AnimStateTransitionNode_2.Transition
  nodes [Result(bCanEnterTransition=Get Sneeze Animation() -> (Sneeze Animation:bool))]
 }
 Transition {path=AnimGraph.AnimGraphNode_StateMachine_1.Idle and Chest_Open.AnimStateTransitionNode_3.Transition, nodes [Result()]}
 EventGraph {nodes [
   Event BlueprintUpdateAnimation()
   > n1 = Branch(Condition=Get bShouldDisplayFrontendReactivity() -> (bShouldDisplayFrontendReactivity:bool))
   > Do Once()
   > n2 = Delay(Duration=4)
   Set Sneeze Animation(execute=[@n4.then, @n2.then], Sneeze Animation:bool=true) -> (Output_Get:bool)
   n3 = Gravity Override State Function DEPRECATED!(
    execute=@n1.else
    CharacterPartAnimInstance=Self-Reference()
    GameTimeInSeconds=GetGameTimeInSeconds()
    GravityOverrideParams_Skydive_GoingUp_28_#13:Gravity Override Setting Struct Structure=(WindFrequency_28_#14=(X=1.5,Y=7,Z=8),WindAmplitude_31_#15=(X=200,Y=-500,Z=400),WindOffset_32_#16=(X=0,Y=0,Z=0),GravityOverride_15_#17=(X=0,Y=-300,Z=1500),JointName_27_#18="root")
    GravityOverrideParams_Skydive_FloatingDown_29_#19:Gravity Override Setting Struct Structure=Get Gravity Override Params Skydive Floating Down() -> (Gravity Override Params Skydive Floating Down:Gravity Override Setting Struct Structure)
    GravityOverrideParams_Skydive_DivingDown_30_#20:Gravity Override Setting Struct Structure=Get Gravity Override Params Skydive Diving Down() -> (Gravity Override Params Skydive Diving Down:Gravity Override Setting Struct Structure)
    GravityOverrideParams_Skydive_Parachuting_31_#21:Gravity Override Setting Struct Structure=Get Gravity Override Params Skydive Parachuting() -> (Gravity Override Params Skydive Parachuting:Gravity Override Setting Struct Structure)
    GravityOverrideParams_OnGround_Standing_32_#22:Gravity Override Setting Struct Structure=(WindFrequency_28_#14=(X=0,Y=0,Z=0),WindAmplitude_31_#15=(X=0,Y=0,Z=0),WindOffset_32_#16=(X=0,Y=0,Z=0),GravityOverride_15_#17=(X=0,Y=0,Z=-980),JointName_27_#18="root")
    GravityOverrideParams_OnGround_Moving_33_#23:Gravity Override Setting Struct Structure=(WindFrequency_28_#14=(X=0,Y=0,Z=0),WindAmplitude_31_#15=(X=0,Y=0,Z=0),WindOffset_32_#16=(X=0,Y=0,Z=0),GravityOverride_15_#17=(X=0,Y=0,Z=-980),JointName_27_#18="root")
    GravityOverrideParams_EmoteOrMelee_34_#24:Gravity Override Setting Struct Structure=(WindFrequency_28_#14=(X=0,Y=0,Z=0),WindAmplitude_31_#15=(X=0,Y=0,Z=0),WindOffset_32_#16=(X=0,Y=0,Z=0),GravityOverride_15_#17=(X=0,Y=0,Z=-980),JointName_27_#18="root")
    GravityOverrideParams_Swimming_35_#25:Gravity Override Setting Struct Structure=(WindFrequency_28_#14=(X=0,Y=0,Z=0),WindAmplitude_31_#15=(X=0,Y=0,Z=0),WindOffset_32_#16=(X=0,Y=0,Z=0),GravityOverride_15_#17=(X=-800,Y=0,Z=0),JointName_27_#18="pelvis")
    GravityOverrideParams_Front_End_36_#26:Gravity Override Setting Struct Structure=(WindFrequency_28_#14=(X=0,Y=0,Z=0),WindAmplitude_31_#15=(X=0,Y=0,Z=0),WindOffset_32_#16=(X=0,Y=0,Z=0),GravityOverride_15_#17=(X=0,Y=0,Z=-980),JointName_27_#18="root")
    GravityOverrideParams_DBNO_40_#27:Gravity Override Setting Struct Structure=(WindFrequency_28_#14=(X=0,Y=0,Z=0),WindAmplitude_31_#15=(X=0,Y=0,Z=0),WindOffset_32_#16=(X=0,Y=0,Z=0),GravityOverride_15_#17=(X=-980,Y=400,Z=0),JointName_27_#18="pelvis")
    GravityOverrideParams_DBNO_Carried_41_#28:Gravity Override Setting Struct Structure=(WindFrequency_28_#14=(X=0,Y=0,Z=0),WindAmplitude_31_#15=(X=0,Y=0,Z=0),WindOffset_32_#16=(X=0,Y=0,Z=0),GravityOverride_15_#17=(X=-980,Y=500,Z=-500),JointName_27_#18="pelvis")
    GravityOverrideParams_OnGround_Falling_45_#29:Gravity Override Setting Struct Structure=(WindFrequency_28_#14=(X=0,Y=0,Z=0),WindAmplitude_31_#15=(X=0,Y=0,Z=0),WindOffset_32_#16=(X=0,Y=0,Z=0),GravityOverride_15_#17=(X=0,Y=0,Z=-980),JointName_27_#18="root")
    GravityOverrideParams_Ziplining_46_#30:Gravity Override Setting Struct Structure=(WindFrequency_28_#14=(X=0,Y=0,Z=0),WindAmplitude_31_#15=(X=0,Y=0,Z=0),WindOffset_32_#16=(X=0,Y=0,Z=0),GravityOverride_15_#17=(X=0,Y=0,Z=-980),JointName_27_#18="root")
    GravityOverrideParams_Skydive_Surfing_49_#31:Gravity Override Setting Struct Structure=(WindFrequency_28_#14=(X=1.5,Y=7,Z=8),WindAmplitude_31_#15=(X=200,Y=-500,Z=400),WindOffset_32_#16=(X=0,Y=0,Z=0),GravityOverride_15_#17=(X=500,Y=-300,Z=-980),JointName_27_#18="root")
    GravityOverrideParams_Riding_Standing_54_#32:Gravity Override Setting Struct Structure=(WindFrequency_28_#14=(X=0,Y=0,Z=0),WindAmplitude_31_#15=(X=0,Y=0,Z=0),WindOffset_32_#16=(X=0,Y=0,Z=0),GravityOverride_15_#17=(X=0,Y=-350,Z=-980),JointName_27_#18="root")
    GravityOverrideParams_Riding_Moving_55_#33:Gravity Override Setting Struct Structure=(WindFrequency_28_#14=(X=0,Y=0,Z=0),WindAmplitude_31_#15=(X=0,Y=0,Z=0),WindOffset_32_#16=(X=0,Y=0,Z=0),GravityOverride_15_#17=(X=0,Y=-450,Z=-450),JointName_27_#18="root")
    GravityOverrideParams_Riding_Moving_Targeting_58_#34:Gravity Override Setting Struct Structure=(WindFrequency_28_#14=(X=0,Y=0,Z=0),WindAmplitude_31_#15=(X=0,Y=0,Z=0),WindOffset_32_#16=(X=0,Y=0,Z=0),GravityOverride_15_#17=(X=0,Y=-200,Z=-980),JointName_27_#18="root")
    GravityOverrideParams_Tactical_Sprint_57_#35:Gravity Override Setting Struct Structure=(WindFrequency_28_#14=(X=0,Y=0,Z=0),WindAmplitude_31_#15=(X=0,Y=0,Z=0),WindOffset_32_#16=(X=0,Y=0,Z=0),GravityOverride_15_#17=(X=0,Y=-450,Z=-500),JointName_27_#18="root")
    GravityOverrideParams_Rail_Grinding_61_#36:Gravity Override Setting Struct Structure=(WindFrequency_28_#14=(X=0,Y=0,Z=0),WindAmplitude_31_#15=(X=0,Y=0,Z=0),WindOffset_32_#16=(X=0,Y=0,Z=0),GravityOverride_15_#17=(X=0,Y=0,Z=-980),JointName_27_#18="root")
    GravityOverrideParams_DBNO_Tumbling_67_#37:Gravity Override Setting Struct Structure=(WindFrequency_28_#14=(X=0,Y=0,Z=0),WindAmplitude_31_#15=(X=0,Y=0,Z=0),WindOffset_32_#16=(X=0,Y=0,Z=0),GravityOverride_15_#17=(X=-2000,Y=0,Z=0),JointName_27_#18="pelvis")
    GravityOverrideParams_DBNO_RollFwd_68_#38:Gravity Override Setting Struct Structure=(WindFrequency_28_#14=(X=0,Y=0,Z=0),WindAmplitude_31_#15=(X=0,Y=0,Z=0),WindOffset_32_#16=(X=0,Y=0,Z=0),GravityOverride_15_#17=(X=-2000,Y=0,Z=0),JointName_27_#18="pelvis")
    GravityOverrideParams_DBNO_RollBwd_69_#39:Gravity Override Setting Struct Structure=(WindFrequency_28_#14=(X=0,Y=0,Z=0),WindAmplitude_31_#15=(X=0,Y=0,Z=0),WindOffset_32_#16=(X=0,Y=0,Z=0),GravityOverride_15_#17=(X=-4000,Y=0,Z=0),JointName_27_#18="pelvis")
    GravityOverrideParams_DBNO_Impulse_70_#40:Gravity Override Setting Struct Structure=(WindFrequency_28_#14=(X=0,Y=0,Z=0),WindAmplitude_31_#15=(X=0,Y=0,Z=0),WindOffset_32_#16=(X=0,Y=0,Z=0),GravityOverride_15_#17=(X=-2000,Y=0,Z=0),JointName_27_#18="pelvis")
    RBAN Params
    DebugDraw
    DebugJointName
   ) -> (then, GravityOverride, CurrentStateOut, GravityOverrideSettingsOut, RBAN SettingsOut)
   > Set Gravity Override(Gravity Override=@n3)
   AnimNotify_Entered_ChestOpen() {is=title K2Node_Event}
   > Set Sneeze Animation(Sneeze Animation=false)
   > n4 = Delay(Duration=13.33)
   n5 = Event BlueprintProcessAnimInputEvent()
   > Branch(Condition=OR Boolean(A=IsNamed(self=@n5.AnimInputEvent, InEventName=aie.ChestOpened), B=IsNamed(self=@n5.AnimInputEvent, InEventName=aie.EnemyPlayerEliminated)))
   > Set Sneeze Animation(Sneeze Animation=true)
   Event BlueprintPostEvaluateAnimation() {is=title K2Node_Event}
   > Set Sneeze Animation(Sneeze Animation=false)
   TryGetPawnOwner()
  ]}
]
variables [
 Gravity Override {type=Vector, default=<0 0 -980>}
 Gravity Override Params Skydive Diving Down {
  type=GravityOverrideSettingStruct
  default {jointName=root, gravityOverride=<-1500 0 0>, windAmplitude=<50 300 100>, windFrequency=<5 30 20>, windOffset=<0 0 0>}
 }
 Gravity Override Params Skydive Floating Down {
  type=GravityOverrideSettingStruct
  default {jointName=root, gravityOverride=<0 -700 1200>, windAmplitude=<200 -500 800>, windFrequency=<4 15 20>, windOffset=<0 -600 400>}
 }
 Gravity Override Params Skydive Parachuting {
  type=GravityOverrideSettingStruct
  default {jointName=root, gravityOverride=<0 -200 -100>, windAmplitude=<50 -150 100>, windFrequency=<1.5 7 8>, windOffset=<0 0 0>}
 }
 Sneeze Animation {type=boolean}
 Timer {type=number}
]
